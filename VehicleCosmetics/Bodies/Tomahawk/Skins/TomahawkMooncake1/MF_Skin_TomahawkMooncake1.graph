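# Material function for the Mooncake1 skin of the Tomahawk vehicle body, setting base color, metallic and roughness.
# Texture masks blend dark, gold, red and off-white, with reflection-noise glints and a gold pattern from stepped-time texture rotation.
# Fist and text decals, then Perlin-animated masks, including a mirrored, distorted UV1 layer, go over that.
# A noise-normal Fresnel finally darkens the color in hard, clamped bands.
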MaterialFunction /VehicleCosmetics/Bodies/Tomahawk/Skins/TomahawkMooncake1/MF_Skin_TomahawkMooncake1
graph {nodes [
  n1 = Constant3Vector() {constant=<rgba 0.0111 0.0119 0.0135 0>}
  n2 = Constant3Vector() {constant=<rgba 1 0.666 0.043 1>}
  n3 = Add(A=Multiply(A=TextureCoordinate() {coordinateIndex=1}) {constB=1.5}, B=Constant2Vector() {g=-0.65, r=-0.25})
  n4 = Time() {bOverride_Period=true, period=6000}
  n5 = Floor(None=Fmod(A=Multiply(A=@n4) {constB=12}, B=Constant() {r=60}))
  n6 = TextureSample(UVs=Rotator(Coordinate=@n3, Time=@n5) {speed=0.2}) {texture=&/T_TomahawkMooncake_RGB2}
  n7 = TextureSample(UVs=Rotator(Coordinate=@n3, Time=@n5) {speed=0.5}) {texture=&T_TomahawkMooncake_RGB2}
  n8 = Saturate(None=LinearInterpolate(B=ConstantBiasScale(None=@n6.G) {bias=-0.1, scale=1.5}, Alpha=ConstantBiasScale(None=@n7.G) {bias=-0.2, scale=5}) {constA=1})
  n9 = Fresnel() {baseReflectFraction=0, exponent=2}
  n10 = TextureCoordinate() {uTiling=0.65}
  n11 = TextureSample(UVs=AppendVector(A=ComponentMask(None=@n10) {g=true}, B=ComponentMask(None=@n10) {r=true})) {samplerType=SAMPLERTYPE_Normal, texture=/VehicleCosmetics/Textures/UniversalTextures/Universal_Noise_Normal_30}
  n12 = ConstantBiasScale(None=OneMinus(None=@n11.G)) {bias=-0.5, scale=2}
  n13 = Multiply(A=Multiply(A=ConstantBiasScale(None=@n9) {bias=0.1, scale=5}, B=Saturate(None=@n12))) {constB=0.01}
  n14 = ConstantBiasScale(None=ReflectionVectorWS()) {bias=0}
  n15 = AppendVector(
   A=TextureSample(UVs=ComponentMask(None=@n14) {g=true, r=true}) {texture=/VehicleCosmetics/Textures/UniversalTextures/Universal_Noise_47_RGB}.R
   B=Constant() {r=0.6}
  )
  n16 = Multiply(
   A=OneMinus(None=@n9)
   B=TextureSample(UVs=Add(A=@n13, B=@n15)) {texture=&Universal_Noise_47_RGB}.B
  )
  n17 = TextureSample(UVs=Add(A=Multiply(A=@n13) {constB=0.5}, B=@n15)) {texture=&Universal_Noise_47_RGB}
  n18 = Max(
   A=Clamp(None=ConstantBiasScale(None=@n16) {bias=-0.2, scale=50}) {maxDefault=0.075}
   B=Clamp(None=ConstantBiasScale(None=@n17.B) {bias=-0.3, scale=50}) {maxDefault=0.125}
  )
  n19 = TextureSample() {texture=&/T_TomahawkMooncake_RGB}
  n20 = Constant3Vector() {constant=<rgba 0.8751 0.8592 0.8355 0>}
  n21 = LinearInterpolate(A=LinearInterpolate(A=@n1, B=Add(A=@n18, B=@n2), Alpha=@n19.G), B=@n20, Alpha=@n19.B)
  n22 = Add(A=@n18, B=Constant3Vector() {constant=<rgba 0.883 0.0161 0.0242 1>})
  n23 = TextureCoordinate() {coordinateIndex=1, uTiling=0.9}
  n24 = TextureSample(UVs=Add(A=Multiply(A=@n23) {constB=5.5}, B=Constant2Vector() {g=-3.65, r=-1.95})) {texture=&/T_TomahawkMooncake_Fist}
  n25 = LinearInterpolate(
   A=LinearInterpolate(A=LinearInterpolate(A=@n1, B=@n2, Alpha=@n8), B=LinearInterpolate(A=@n21, B=@n22, Alpha=@n19.R), Alpha=@n19.A)
   B=@n24.RGB
   Alpha=@n24.RGB
  )
  n26 = TextureCoordinate() {coordinateIndex=1, uTiling=0.9}
  n27 = TextureSample(UVs=Add(A=Multiply(A=@n26) {constB=4.5}, B=Constant2Vector() {g=-2.9, r=-1.525})) {texture=&/T_TomahawkMooncake_Text}
  n28 = TextureSample() {texture=&T_TomahawkMooncake_RGB2}
  n29 = Constant4Vector() {constant=<rgba 2 0.05 3 0.25>}
  n30 = MaterialFunctionCall!(Texture Coordinates, Tiling X=@n29.R, Tiling Y=@n29.R, PanSpeedX=@n29.R, PanSpeedY=@n29.R, Angle, RotateSpeed, Balance, Chaos=@n29.R) {materialFunction=/VehicleCosmetics/Materials/MaterialFunctions/MF_Noisemaker_Perlin}
  n31 = Multiply(A=Multiply(A=Multiply(A=@n28.B, B=@n19.A), B=Multiply(A=@n30.Perlin1) {constB=0.5})) {constB=4}
  n32 = Constant4Vector() {constant=<rgba 0.75 0.2 2 0.5>}
  n33 = MaterialFunctionCall(Tiling X=@n32.B, Tiling Y=@n32.B, PanSpeedX=@n32.B, PanSpeedY=@n32.B) {materialFunction=&MF_Noisemaker_Perlin}
  n34 = MaterialFunctionCall(Tiling X=@n32.B, Tiling Y=@n32.B, PanSpeedX=@n32.B, PanSpeedY=@n32.B) {materialFunction=&MF_Noisemaker_Perlin}
  n35 = Multiply(A=ConstantBiasScale(None=AppendVector(A=@n33.Perlin1, B=@n34.Perlin1)) {bias=-0.5, scale=2}) {constB=0.0075}
  n36 = TextureCoordinate() {coordinateIndex=1}
  n37 = OneMinus(None=ComponentMask(None=@n36) {g=true})
  n38 = ComponentMask(None=@n36) {r=true}
  n39 = Floor(None=Multiply(A=@n38) {constB=2})
  n40 = OneMinus(None=LinearInterpolate(A=@n38, B=OneMinus(None=@n38), Alpha=Saturate(None=@n39)))
  n41 = Add(A=AppendVector(A=@n37, B=ConstantBiasScale(None=@n40) {bias=-0.5, scale=1}), B=Constant2Vector() {g=0.25, r=-0.35})
  n42 = Multiply(
   A=TextureSample(UVs=Add(A=@n35, B=@n41)) {texture=&T_TomahawkMooncake_RGB2}.R
   B=Saturate(None=ConstantBiasScale(None=@n37) {bias=-0.4, scale=5000})
  )
  n43 = LinearInterpolate(
   A=LinearInterpolate(A=LinearInterpolate(A=@n25, B=@n27.RGB, Alpha=@n27.RGB), B=@n1, Alpha=Saturate(None=@n31))
   B=@n1
   Alpha=Saturate(None=Multiply(A=@n42) {constB=2})
  )
  n44 = Saturate(None=ConstantBiasScale(None=@n42) {bias=-0.2, scale=2})
  n45 = ComponentMask(None=@n11.RGB) {g=true, r=true}
  n46 = MaterialFunctionCall!(Fresnel Power=Constant() {r=1}, Normal=AppendVector(A=ConstantBiasScale(None=@n45) {bias=0, scale=0.2}, B=@n11.B)) {materialFunction=/VehicleCosmetics/Materials/MaterialFunctions/MF_Fresnel_Prime}
  n47 = Clamp(None=ConstantBiasScale(None=@n46.Result) {bias=-0.75, scale=50}) {maxDefault=0.925}
  n48 = ConstantBiasScale(None=@n46.Result) {bias=-0.5, scale=10}
  n49 = Clamp(None=@n48) {maxDefault=0.8}
  n50 = SetMaterialAttributes!(
   MaterialAttributes
   BaseColor=LinearInterpolate(A=LinearInterpolate(A=@n43, B=@n20, Alpha=@n44), B=@n1, Alpha=Max(A=@n47, B=@n49))
   Metallic=Constant() {r=0.2}
   Roughness=Constant() {r=0.2}
   EmissiveColor
   Normal
   AmbientOcclusion
  ) {attributeSetTypes [
    #1
    #2
    #3
    #4
    #5
    #6
   ]}
  FunctionOutput(None=@n50)
  Saturate(None=@n48)
 ]}
